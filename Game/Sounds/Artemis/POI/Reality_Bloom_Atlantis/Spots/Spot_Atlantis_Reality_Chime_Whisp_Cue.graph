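# Sound cue for a one-shot ambient chime-whisp spot at the Reality Bloom Atlantis POI, using Reality Forest biome waves and tile-ambient attenuation.
# The first two quality levels randomly choose between two volume envelopes (one external curve, one inline), each modulating pitch and volume over a shared six-wave random pool.
# The last two quality levels skip the envelopes and play a four-wave random pool with narrower modulation.
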SoundCue /Game/Sounds/Artemis/POI/Reality_Bloom_Atlantis/Spots/Spot_Atlantis_Reality_Chime_Whisp_Cue
root=QualityLevel_0
nodes [
 {
  id=Enveloper_0
  children [Modulator_1]
  properties {
   pitchCurve {editorCurveData {keys [{time=0}]}}
   volumeCurve {
    externalCurve=/Game/Sounds/Artemis/POI/Reality_Bloom_Atlantis/ExternalCurve_Magic_Spot_Softener_Long
    editorCurveData {keys [
      <at 0 1>
      {arriveTangent=-0.0862, interpMode=RCIM_Cubic, leaveTangent=-0.0862, tangentMode=RCTM_User, time=1.4067, value=0.92}
      {arriveTangent=-0.0912, interpMode=RCIM_Cubic, leaveTangent=-0.0912, tangentMode=RCTM_User, time=2.3018, value=0.4885}
      <at 6.9953 0.3749>
      {arriveTangent=-0.2462, interpMode=RCIM_Cubic, leaveTangent=-0.2462, tangentMode=RCTM_User, time=10.4982, value=0.2}
     ]}
   }
  }
 }
 {
  id=Enveloper_1
  children [Modulator_3]
  properties {
   pitchCurve {editorCurveData {keys [{time=0}]}}
   volumeCurve {editorCurveData {keys [
      <at 0 1>
      {arriveTangent=-0.0862, interpMode=RCIM_Cubic, leaveTangent=-0.0862, tangentMode=RCTM_User, time=1.4067, value=0.92}
      {arriveTangent=-0.0912, interpMode=RCIM_Cubic, leaveTangent=-0.0912, tangentMode=RCTM_User, time=2.3018, value=0.4885}
      {arriveTangent=-0.0242, interpMode=RCIM_Cubic, leaveTangent=-0.0242, time=4.5, value=0.1}
      {arriveTangent=-0.003, interpMode=RCIM_Cubic, leaveTangent=-0.003, tangentMode=RCTM_User, time=6.9953, value=0.3749}
      {arriveTangent=-0.1638, interpMode=RCIM_Cubic, leaveTangent=-0.1638, tangentMode=RCTM_User, time=10.4982, value=0.2}
     ]}}
  }
 }
 {id=Modulator_1, children [Random_0], properties {pitchMax=1.1, pitchMin=0.7, volumeMax=1.1, volumeMin=0.7}}
 {id=Modulator_2, children [Random_1], properties {pitchMax=1.1, volumeMax=1.1, volumeMin=0.9}}
 {id=Modulator_3, children [Random_0], properties {pitchMax=1.1, pitchMin=0.7, volumeMax=1.1, volumeMin=0.7}}
 {id=QualityLevel_0, children [Random_2,Random_2,Modulator_2,Modulator_2]}
 {id=Random_0, children [WavePlayer_0,WavePlayer_1,WavePlayer_2,WavePlayer_3,WavePlayer_4,WavePlayer_5], properties {weights [1,1,1,1,1,1]}}
 {id=Random_1, children [WavePlayer_6,WavePlayer_7,WavePlayer_8,WavePlayer_9], properties {weights [1,1,1,1]}}
 {id=Random_2, children [Enveloper_0,Enveloper_1], properties {weights [1,1]}}
 {id=WavePlayer_0, properties {soundWaveAssetPtr=/Game/Sounds/Artemis/Biomes/Reality_Forest/Spots/AMB_Biome_Reality_Forest_SS_ChimeWhisp_01}}
 {id=WavePlayer_1, properties {soundWaveAssetPtr=/Game/Sounds/Artemis/Biomes/Reality_Forest/Spots/AMB_Biome_Reality_Forest_SS_ChimeWhisp_02}}
 {id=WavePlayer_2, properties {soundWaveAssetPtr=/Game/Sounds/Artemis/Biomes/Reality_Forest/Spots/AMB_Biome_Reality_Forest_SS_ChimeWhisp_03}}
 {id=WavePlayer_3, properties {soundWaveAssetPtr=/Game/Sounds/Artemis/Biomes/Reality_Forest/Spots/AMB_Biome_Reality_Forest_SS_ChimeWhisp_04}}
 {id=WavePlayer_4, properties {soundWaveAssetPtr=/Game/Sounds/Artemis/Biomes/Reality_Forest/Spots/AMB_Biome_Reality_Forest_SS_ChimeWhisp_05}}
 {id=WavePlayer_5, properties {soundWaveAssetPtr=/Game/Sounds/Artemis/Biomes/Reality_Forest/Spots/AMB_Biome_Reality_Forest_SS_ChimeWhisp_06}}
 {id=WavePlayer_6, properties {soundWaveAssetPtr=&AMB_Biome_Reality_Forest_SS_ChimeWhisp_01}}
 {id=WavePlayer_7, properties {soundWaveAssetPtr=&AMB_Biome_Reality_Forest_SS_ChimeWhisp_02}}
 {id=WavePlayer_8, properties {soundWaveAssetPtr=&AMB_Biome_Reality_Forest_SS_ChimeWhisp_03}}
 {id=WavePlayer_9, properties {soundWaveAssetPtr=&AMB_Biome_Reality_Forest_SS_ChimeWhisp_04}}
]
properties {
 attenuationSettings=/Game/Sounds/Attenuation/Fort_TileAmbient_OneShot_Wind_Att
 bEnableBusSends=false
 maxDistance=8000
 pitchMultiplier=0.9
 soundClassObject=&Fort_Ambient_OneShots
 volumeMultiplier=0.501
 preEffectBusSends [{}]
}
