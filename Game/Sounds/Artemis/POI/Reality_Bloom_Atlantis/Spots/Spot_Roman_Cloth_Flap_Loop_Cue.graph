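# Looping outdoor ambient spot cue of flapping cloth for the Reality Bloom Atlantis POI, reusing a PirateCove sail-flap wave.
# A short random delay starts three nested looping envelopers: 40- and 60-second volume curves and a 60-second pitch curve, layering gust-like variation onto one looping wave.
# A quality node feeds that wave to its first two levels, leaving two empty.
# Overrides set capsule attenuation with occlusion, listener focus and manual reverb, plus tight concurrency.
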SoundCue /Game/Sounds/Artemis/POI/Reality_Bloom_Atlantis/Spots/Spot_Roman_Cloth_Flap_Loop_Cue
root=Delay_0
nodes [
 {id=Delay_0, children [Enveloper_4], properties {delayMax=0.75, delayMin=0.2}}
 {
  id=Enveloper_0
  children [QualityLevel_0]
  properties {
   bLoop=true
   bLoopIndefinitely=true
   loopEnd=60
   pitchCurve {editorCurveData {keys [
      {interpMode=RCIM_Cubic, time=0}
      <at 15 0.85 RCIM_Cubic>
      {arriveTangent=0.0091, interpMode=RCIM_Cubic, leaveTangent=0.0091, tangentMode=RCTM_User, time=30}
      <at 45 1.1 RCIM_Cubic>
      {time=60}
     ]}}
   volumeCurve {editorCurveData {keys [<at 0 1>]}}
  }
 }
 {
  id=Enveloper_3
  children [Enveloper_0]
  properties {
   bLoop=true
   bLoopIndefinitely=true
   loopEnd=60
   pitchCurve {editorCurveData {keys [{time=0}]}}
   volumeCurve {editorCurveData {keys [
      {arriveTangent=-0.01, interpMode=RCIM_Cubic, time=0, value=1}
      {arriveTangent=-0.0035, interpMode=RCIM_Cubic, leaveTangent=-0.0035, tangentMode=RCTM_User, time=5.0017, value=0.3634}
      {arriveTangent=0.0016, interpMode=RCIM_Cubic, leaveTangent=0.0016, tangentMode=RCTM_User, time=7.591, value=0.8382}
      {arriveTangent=0.0008, interpMode=RCIM_Cubic, leaveTangent=0.0008, tangentMode=RCTM_User, time=11.5666, value=0.3768}
      {arriveTangent=0.0016, interpMode=RCIM_Cubic, leaveTangent=0.0016, tangentMode=RCTM_User, time=13.4364, value=0.3624}
      {arriveTangent=0.0009, interpMode=RCIM_Cubic, leaveTangent=0.0009, time=15.1685, value=0.7207}
      {arriveTangent=-0.0011, interpMode=RCIM_Cubic, leaveTangent=-0.0011, tangentMode=RCTM_User, time=19.0457, value=0.3674}
      {arriveTangent=0.0002, interpMode=RCIM_Cubic, leaveTangent=0.0002, time=24.1334, value=0.8568}
      {arriveTangent=0.0016, interpMode=RCIM_Cubic, leaveTangent=0.0016, tangentMode=RCTM_User, time=28.493, value=0.3696}
      {arriveTangent=-0.0003, interpMode=RCIM_Cubic, leaveTangent=-0.0003, time=31.406, value=0.7353}
      {arriveTangent=0.0009, interpMode=RCIM_Cubic, leaveTangent=0.0009, tangentMode=RCTM_User, time=35.3816, value=0.3676}
      {arriveTangent=-0.0008, interpMode=RCIM_Cubic, leaveTangent=-0.0008, time=38.1075, value=0.8568}
      <at 42.4671 0.3623>
      {arriveTangent=0.0024, interpMode=RCIM_Cubic, leaveTangent=0.0024, tangentMode=RCTM_User, time=45.0257, value=0.3674}
      {arriveTangent=-0.0006, interpMode=RCIM_Cubic, leaveTangent=-0.0006, time=48.1356, value=0.8382}
      {arriveTangent=-0.0056, interpMode=RCIM_Cubic, leaveTangent=-0.0056, tangentMode=RCTM_User, time=53.5932, value=0.3624}
      <at 55.5199 0.7454>
      {arriveTangent=-0.0003, interpMode=RCIM_Cubic, leaveTangent=-0.0003, tangentMode=RCTM_User, time=58.2125, value=0.3621}
      <at 60 1 RCIM_Cubic>
     ]}}
  }
 }
 {
  id=Enveloper_4
  children [Enveloper_3]
  properties {
   bLoop=true
   bLoopIndefinitely=true
   loopEnd=40
   pitchCurve {editorCurveData {keys [{time=0}]}}
   volumeCurve {editorCurveData {keys [
      {arriveTangent=-0.01, interpMode=RCIM_Cubic, time=0, value=1}
      <at 9.9146 0.5569 RCIM_Cubic>
      {arriveTangent=-0.0028, interpMode=RCIM_Cubic, leaveTangent=-0.0028, time=20, value=1}
      {arriveTangent=-0.0749, interpMode=RCIM_Cubic, leaveTangent=-0.0749, tangentMode=RCTM_User, time=26.0936, value=0.5122}
      {arriveTangent=0.129, interpMode=RCIM_Cubic, leaveTangent=0.129, tangentMode=RCTM_User, time=36.3285, value=0.6555}
      <at 40 1 RCIM_Cubic>
     ]}}
  }
 }
 {id=QualityLevel_0, children [WavePlayer_0,WavePlayer_0,null,null]}
 {id=WavePlayer_0, properties {bLooping=true, soundWaveAssetPtr=/Game/Athena/Sounds/Ambient/PirateCove/Amb_PirateCove_SailFlap_01}}
]
properties {
 attenuationSettings=&/Spot_Cloth_Flag_Flapping_Bigger_Att
 bOverrideAttenuation=true
 bOverrideConcurrency=true
 maxDistance=2575
 pitchMultiplier=0.9
 soundClassObject=&Fort_Ambient_Outdoors_Athena
 volumeMultiplier=0.501
 attenuationOverrides {
  attenuationShape=Capsule
  bEnableFocusInterpolation=true
  bEnableListenerFocus=true
  bEnableOcclusion=true
  falloffDistance=2500
  focusAttackInterpSpeed=3
  focusReleaseInterpSpeed=3
  focusVolumeAttenuation=1.1
  manualReverbSendLevel=10
  nonFocusVolumeAttenuation=0.5
  occlusionLowPassFilterFrequency=3000
  occlusionTraceChannel=ECC_GameTraceChannel11
  reverbDistanceMax=2000
  reverbSendMethod=Manual
  reverbWetLevelMax=5
  attenuationShapeExtents=<0 75>
 }
 concurrencyOverrides {maxCount=2, voiceStealReleaseTime=2}
}
